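# Editor utility dialog for CRSL Maker that maps selected Static Mesh assets to names with editable text (default "SM_") removed.
# On open or Add Selected, it fills the map, clearing it first if OverrideList is set (initially false, though its checkbox shows checked).
# A self-bound DetailsView exposes the Meshes category, and Copy Name takes the removal text from a selected mesh.
# OK broadcasts the map through a dispatcher and closes; Cancel just closes.
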EditorUtilityWidgetBlueprint /AnimInEdTools/CRSL_Maker/EUW_CRSL_AddSelectedSM
graphs [
 EventGraph {nodes [
   n1 = "On Clicked (AddButton)"() {is=title K2Node_ComponentBoundEvent}
   n2 = Sequence!(execute=[@n1.then, @n13.then]) -> (then_0, then_1, then_2:Exec)
   > Branch(Condition=Get OverrideList() -> (OverrideList:bool))
   > Clear!(
    execute
    TargetMap:Map of Static Mesh Object References to Names=Get Meshes List() -> (Meshes List:Map of Static Mesh Object References to Names)
   ) -> (then) {is=title K2Node_CallFunction}
   n3 = GetSelectedAssetsOfClass(execute=@n2.then_1, AssetClass=&StaticMesh)
   > n4 = For Each Loop(Array:Array of Object References=@n3) -> (Array Element:Object Reference)
   > n5 = Cast To StaticMesh(Object=@n4)
   n6 = "To String (Text)"(InText=Get TextToRemove() -> (TextToRemove:Text))
   n7 = GetDisplayName(Object=@n5)
   > Add!(
    execute
    TargetMap:Map of Static Mesh Object References to Names=Get Meshes List()
    Key:"Static Mesh Object Reference (by ref)"=@n5
    Value:"Name (by ref)"=String To Name(InString=Replace(SourceString=@n7, From=@n6, SearchCase=CaseSensitive))
   ) -> (then) {is=title K2Node_CallFunction}
   n8 = "On Check State Changed (OverwriteList)"!() -> (OutputDelegate, then, bIsChecked:bool) {is=title K2Node_ComponentBoundEvent}
   > Set OverrideList(OverrideList:bool=@n8.bIsChecked) -> (Output_Get:bool)
   n9 = "On Text Changed (ToRemoveTextBox)"!() -> (OutputDelegate, then, Text:"Text (by ref)") {is=title K2Node_ComponentBoundEvent}
   > Set TextToRemove(TextToRemove:Text=@n9.Text) -> (Output_Get:Text)
   "On Clicked (CopyNameButton)"() {is=title K2Node_ComponentBoundEvent}
   > n10 = GetSelectedAssetsOfClass(AssetClass=&StaticMesh)
   > n11 = Cast To StaticMesh(Object="Get (a copy)"(Array:Array of Object References=@n10) -> (Output:Object Reference))
   > n12 = Set Text(
    Text="To Text (String)"(InString=GetDisplayName(Object=@n11))
    self:Text Box Object Reference=Get ToRemoveTextBox() -> (ToRemoveTextBox:Editor Utility Editable Text Box Object Reference)
   )
   > Set TextToRemove(TextToRemove=@n12)
   "On Clicked (CancelButton)"() {is=title K2Node_ComponentBoundEvent}
   > RemoveFromParent()
   "On Clicked (OKButton)"() {is=title K2Node_ComponentBoundEvent}
   > Call Static Mesh Map!(execute:Exec, self:EUW CRSL Add Selected SM Object Reference, SM Map:Map of Static Mesh Object References to Names=Get Meshes List()) -> (then:Exec) {is=title K2Node_CallDelegate}
   > RemoveFromParent()
   n13 = Add Selected() {is=title K2Node_CustomEvent}
   Event PreConstruct()
   > SetObject(self=Get DetailsView() -> (DetailsView:Details View Object Reference), NewObject=Self-Reference())
   Event Construct()
   > Add Selected() {is=title K2Node_CallFunction}
   ExpandAll(self:Tree View Object Reference)
  ]}
 Static Mesh Map {nodes [Static Mesh Map!() -> (then, SM Map:Map of Static Mesh Object References to Names) {is=title K2Node_FunctionEntry}]}
]
variables [
 Meshes List {default {}}
 OverrideList {type=boolean}
 Static Mesh Map {default=()}
 TextToRemove {type=string, default=SM_}
]
widgets {
 info {rootWidgetClass=&VerticalBox, widgetCount=24}
 widgets [
  VerticalBox_2 {widgetClassPath=&VerticalBox, widget=VerticalBox_2}
  Border {
   widgetClassPath=&Border
   parent=VerticalBox_2
   widget=Border
   properties {horizontalAlignment=HAlign_Center, verticalAlignment=VAlign_Center, brushColor=<rgba 0 0 0 0.5>, padding=<ltrb 0 0 0 0>}
   slot {class=VerticalBoxSlot, properties {size {sizeRule=Fill, value=1}}}
  }
  SizeBox_0 {
   widgetClassPath=&SizeBox
   parent=Border
   widget=SizeBox_0
   properties {bOverride_MinDesiredWidth=true, maxDesiredHeight=500, maxDesiredWidth=500, minDesiredHeight=320, minDesiredWidth=550}
   slot {class=BorderSlot, properties {horizontalAlignment=HAlign_Center, verticalAlignment=VAlign_Center, padding=<ltrb 0 0 0 0>}}
  }
  Border_1 {
   widgetClassPath=&Border
   parent=SizeBox_0
   widget=Border_1
   properties {brushColor=<rgba 0.0157 0.0157 0.0157 1>, padding=<ltrb 0 0 0 0>}
   slot {class=SizeBoxSlot}
  }
  VerticalBox_3 {widgetClassPath=&VerticalBox, parent=Border_1, widget=VerticalBox_3, slot {class=BorderSlot, properties {padding=<ltrb 0 0 0 0>}}}
  EditorUtilityScrollBox_0 {
   widgetClassPath=&EditorUtilityScrollBox
   parent=VerticalBox_3
   widget=EditorUtilityScrollBox_0
   slot {class=VerticalBoxSlot, properties {size {sizeRule=Fill, value=1}}}
  }
  DetailsView {
   widgetClassPath=&DetailsView
   parent=EditorUtilityScrollBox_0
   widget=DetailsView
   properties {categoriesToShow [Meshes]}
   slot {class=ScrollBoxSlot}
  }
  VerticalBox_0 {widgetClassPath=&VerticalBox, parent=VerticalBox_3, widget=VerticalBox_0, slot {class=VerticalBoxSlot}}
  HorizontalBox_1 {
   widgetClassPath=&HorizontalBox
   parent=VerticalBox_0
   widget=HorizontalBox_1
   slot {class=VerticalBoxSlot, properties {padding=<ltrb 0 0 0 5>, size {sizeRule=Fill, value=1}}}
  }
  AddButton {
   widgetClassPath=&EditorUtilityButton
   parent=HorizontalBox_1
   widget=AddButton
   slot {class=HorizontalBoxSlot, properties {size {sizeRule=Fill, value=1}}}
  }
  TextBlock_45 {
   widgetClassPath=&TextBlock
   parent=AddButton
   widget=TextBlock_45
   properties {
    justification=Center
    text=Add Selected
    font {
     bForceMonospaced=false
     bMaterialIsStencil=false
     fontMaterial=null
     fontObject=&Roboto
     letterSpacing=0
     monospacedWidth=1
     size=13.5
     skewAmount=0
     typefaceFontName=Bold
     outlineSettings {
      bApplyOutlineToDropShadows=false
      bMiteredCorners=false
      bSeparateFillAlpha=false
      outlineMaterial=null
      outlineSize=0
      outlineColor=<rgba 0 0 0 1>
     }
    }
   }
   slot {class=ButtonSlot, properties {horizontalAlignment=HAlign_Fill, verticalAlignment=VAlign_Fill, padding=<ltrb 0 5 0 5>}}
  }
  HorizontalBox_45 {
   widgetClassPath=&HorizontalBox
   parent=VerticalBox_0
   widget=HorizontalBox_45
   slot {class=VerticalBoxSlot, properties {padding=<ltrb 10 0 10 5>}}
  }
  TextBlock_239 {
   widgetClassPath=&TextBlock
   parent=HorizontalBox_45
   widget=TextBlock_239
   properties {
    text=Override Current List?
    font {
     bForceMonospaced=false
     bMaterialIsStencil=false
     fontMaterial=null
     fontObject=&Roboto
     letterSpacing=0
     monospacedWidth=1
     size=9
     skewAmount=0
     typefaceFontName=Regular
     outlineSettings {
      bApplyOutlineToDropShadows=false
      bMiteredCorners=false
      bSeparateFillAlpha=false
      outlineMaterial=null
      outlineSize=0
      outlineColor=<rgba 0 0 0 1>
     }
    }
   }
   slot {class=HorizontalBoxSlot, properties {verticalAlignment=VAlign_Center, padding=<ltrb 0 0 5 0>}}
  }
  OverwriteList {
   widgetClassPath=&EditorUtilityCheckBox
   bIsVariable=true
   parent=HorizontalBox_45
   widget=OverwriteList
   properties {checkedState=Checked}
   slot {class=HorizontalBoxSlot}
  }
  HorizontalBox_44 {
   widgetClassPath=&HorizontalBox
   parent=VerticalBox_0
   widget=HorizontalBox_44
   slot {class=VerticalBoxSlot, properties {padding=<ltrb 10 0 10 5>}}
  }
  TextBlock {
   widgetClassPath=&TextBlock
   parent=HorizontalBox_44
   widget=TextBlock
   properties {
    text=Remove From Name
    font {
     bForceMonospaced=false
     bMaterialIsStencil=false
     fontMaterial=null
     fontObject=&Roboto
     letterSpacing=0
     monospacedWidth=1
     size=9
     skewAmount=0
     typefaceFontName=Regular
     outlineSettings {
      bApplyOutlineToDropShadows=false
      bMiteredCorners=false
      bSeparateFillAlpha=false
      outlineMaterial=null
      outlineSize=0
      outlineColor=<rgba 0 0 0 1>
     }
    }
   }
   slot {class=HorizontalBoxSlot, properties {verticalAlignment=VAlign_Center, padding=<ltrb 0 0 5 0>}}
  }
  ToRemoveTextBox {
   widgetClassPath=&EditorUtilityEditableTextBox
   bIsVariable=true
   parent=HorizontalBox_44
   widget=ToRemoveTextBox
   properties {text=SM_}
   slot {class=HorizontalBoxSlot, properties {size {sizeRule=Fill, value=1}}}
  }
  CopyNameButton {
   widgetClassPath=&EditorUtilityButton
   parent=HorizontalBox_44
   widget=CopyNameButton
   slot {class=HorizontalBoxSlot, properties {padding=<ltrb 10 0 0 0>}}
  }
  TextBlock_356 {
   widgetClassPath=&TextBlock
   parent=CopyNameButton
   widget=TextBlock_356
   properties {
    text=Copy Name
    font {
     bForceMonospaced=false
     bMaterialIsStencil=false
     fontMaterial=null
     fontObject=&Roboto
     letterSpacing=0
     monospacedWidth=1
     size=9
     skewAmount=0
     typefaceFontName=Bold
     outlineSettings {
      bApplyOutlineToDropShadows=false
      bMiteredCorners=false
      bSeparateFillAlpha=false
      outlineMaterial=null
      outlineSize=0
      outlineColor=<rgba 0 0 0 1>
     }
    }
   }
   slot {class=ButtonSlot, properties {padding=<ltrb 10 2 10 2>}}
  }
  HorizontalBox_102 {widgetClassPath=&HorizontalBox, parent=VerticalBox_3, widget=HorizontalBox_102, slot {class=VerticalBoxSlot}}
  OKButton {
   widgetClassPath=&EditorUtilityButton
   parent=HorizontalBox_102
   widget=OKButton
   slot {class=HorizontalBoxSlot, properties {size {sizeRule=Fill, value=1}}}
  }
  TextBlock_77 {
   widgetClassPath=&TextBlock
   parent=OKButton
   widget=TextBlock_77
   properties {
    text=OK
    font {
     bForceMonospaced=false
     bMaterialIsStencil=false
     fontMaterial=null
     fontObject=&Roboto
     letterSpacing=0
     monospacedWidth=1
     size=13.5
     skewAmount=0
     typefaceFontName=Bold
     outlineSettings {
      bApplyOutlineToDropShadows=false
      bMiteredCorners=false
      bSeparateFillAlpha=false
      outlineMaterial=null
      outlineSize=0
      outlineColor=<rgba 0 0 0 1>
     }
    }
   }
   slot {class=ButtonSlot}
  }
  CancelButton {
   widgetClassPath=&EditorUtilityButton
   parent=HorizontalBox_102
   widget=CancelButton
   slot {class=HorizontalBoxSlot, properties {size {sizeRule=Fill, value=1}}}
  }
  TextBlock_140 {
   widgetClassPath=&TextBlock
   parent=CancelButton
   widget=TextBlock_140
   properties {
    text=Cancel
    font {
     bForceMonospaced=false
     bMaterialIsStencil=false
     fontMaterial=null
     fontObject=&Roboto
     letterSpacing=0
     monospacedWidth=1
     size=13.5
     skewAmount=0
     typefaceFontName=Bold
     outlineSettings {
      bApplyOutlineToDropShadows=false
      bMiteredCorners=false
      bSeparateFillAlpha=false
      outlineMaterial=null
      outlineSize=0
      outlineColor=<rgba 0 0 0 1>
     }
    }
   }
   slot {class=ButtonSlot}
  }
 ]
}
