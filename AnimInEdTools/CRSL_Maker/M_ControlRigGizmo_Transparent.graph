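# Translucent, unlit material for Control Rig gizmos, with ray traced shadows off.
# A float3 built from the red channel of the Color parameter in X and Z is divided by eye adaptation to offset auto exposure.
# World vertex normals are scaled by a Normal Offset Mult parameter, and a separate Opacity scalar is defined.
# The graph shows no links from these chains to material outputs.
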Material /AnimInEdTools/CRSL_Maker/M_ControlRigGizmo_Transparent
graph {
 properties {bCastRayTracedShadows=false, blendMode=BLEND_Translucent, pixelDepthOffsetMode=PDOM_Legacy, shadingModel=MSM_Unlit, usedShadingModels=Unlit}
 nodes [
  n1 = VectorParameter() {parameterName=Color, defaultValue=<rgba 1 1 1 0.1>}
  Divide(
   A=MaterialFunctionCall(X=@n1.R, Z=@n1.R) {materialFunction=&MakeFloat3}.Result
   B=EyeAdaptation().EyeAdaptation
  )
  Multiply(A=VertexNormalWS(), B=ScalarParameter() {defaultValue=0.1, parameterName=Normal Offset Mult}) {constB=2}
  ScalarParameter() {parameterName=Opacity}
 ]
}
