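# Editor utility widget that builds and saves a Control Rig Shape Library from a skeletal mesh, Atom Model meshes and custom shapes.
# Populate Meshes maps the model's processed static meshes to selection-set names.
# Run offsets each mesh by its bone's inverse world transform, read through Geometry Script, lets custom shapes override, then uses Python to pick the project's gizmo material.
# Inputs persist in a CRSLM asset or save-game slot.
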EditorUtilityWidgetBlueprint /AnimInEdTools/CRSL_Maker/EUW_CRSL_Maker
graphs [
 EventGraph {nodes [
   n1 = LoadAsset(AssetPath=/JunoBase/Materials/M_ControlRigGizmo)
   > SetEditorProperty(
    Object=Get CR Shape Lib() -> (CR Shape Lib:Control Rig Shape Library Object Reference)
    PropertyName=DefaultMaterial
    PropertyValue:Soft Object Reference=Make Soft Reference!(Input:Object Reference=@n1) -> (Output:Soft Object Reference)
   )
   "On Clicked (Button_Run)"() {is=title K2Node_ComponentBoundEvent}
   > n2 = Sequence!(execute) -> (then_0, then_1, then_2:Exec, then_3:Exec, then_4:Exec, then_5:Exec)
   n3 = CreateDynamicMeshPool!(execute=@n2.then_1) -> (then, ReturnValue:Dynamic Mesh Pool Object Reference)
   > n4 = RequestMesh!(execute, self:Dynamic Mesh Pool Object Reference=@n3) -> (then, ReturnValue:Dynamic Mesh Object Reference)
   > n5 = Set Dynamic Mesh(Dynamic Mesh:Dynamic Mesh Object Reference=@n4) -> (Output_Get:Dynamic Mesh Object Reference)
   > CopyMeshFromSkeletalMesh!(
    execute
    FromSkeletalMeshAsset:Skeletal Mesh Object Reference=Get SKM() -> (SKM:Skeletal Mesh Object Reference)
    ToDynamicMesh:Dynamic Mesh Object Reference=@n5
    AssetOptions:Geometry Script Copy Mesh From Asset Options Structure
    RequestedLOD:Geometry Script Mesh Read LOD Structure=Make Geometry Script Mesh Read LOD!(LODType:EGeometryScriptLODType Enum=SourceModel, LODIndex:int=0) -> (GeometryScriptMeshReadLOD:Geometry Script Mesh Read LOD Structure) {is=title K2Node_MakeStruct}
   ) -> (Failure:Exec, Success:Exec, ReturnValue:Dynamic Mesh Object Reference)
   n6 = Keys!(
    execute=@n2.then_2
    TargetMap:Map of Names to Static Mesh Soft Object References=Get Static Meshes() -> (Static Meshes:Map of Names to Static Mesh Soft Object References)
   ) -> (then, Keys:Array of Names)
   > n7 = For Each Loop(Array:Array of Names=@n6) -> (Array Element:Name)
   > n8 = Keys(TargetMap:Map of Names to Struct Custom Shape Structures=Get Custom Shapes() -> (Custom Shapes:Map of Names to Struct Custom Shape Structures))
   > Branch(Condition="integer < integer"(A=Find Item!(TargetArray:Array of Names=@n8, ItemToFind:"Name (by ref)"=@n7) -> (ReturnValue)))
   > n9 = GetBoneInfo!(execute, TargetMesh:Dynamic Mesh Object Reference=Get Dynamic Mesh() -> (Dynamic Mesh:Dynamic Mesh Object Reference), BoneName:Name=@n7) -> (then, bIsValidBoneName:bool, BoneInfo:Bone Info Structure, ReturnValue:Dynamic Mesh Object Reference)
   n10 = Get Static Meshes()
   n11 = Break Bone Info!(GeometryScriptBoneInfo:"Bone Info Structure (by ref)"=@n9.BoneInfo) -> (Index:int, Name:Name, ParentIndex:int, LocalTransform:Transform, WorldTransform:Transform, Color:Linear Color Structure) {is=title K2Node_BreakStruct}
   > Add(
    TargetArray:Array of Shape Structures=Get Control Rig Shape Definition() -> (Control Rig Shape Definition:Array of Shape Structures)
    NewItem:"Shape Structure (by ref)"=Make Shape!(
     ShapeName:Name=@n7
     Transform:Transform=InvertTransform!(T:"Transform (by ref)"=@n11.WorldTransform) -> (ReturnValue:Transform)
     ShapeProxy:Soft Object Reference=Find(TargetMap:Map of Names to Static Mesh Soft Object References=@n10, Key:"Name (by ref)"=@n7) -> (Value:Static Mesh Soft Object Reference)
    ) -> (ControlRigShapeDefinition:Shape Structure) {is=title K2Node_MakeStruct}
   ) {is=title K2Node_CallArrayFunction}
   n12 = Keys(execute=@n2.then_3, TargetMap=Get Custom Shapes())
   > n13 = For Each Loop(Array=@n12)
   n14 = Break Struct Custom Shape!(Struct_CustomShape:"Struct Custom Shape Structure (by ref)"=Find(TargetMap:Map of Names to Struct Custom Shape Structures=Get Custom Shapes(), Key=@n13) -> (Value:Struct Custom Shape Structure)) -> (StaticMesh_10_#1:Static Mesh Soft Object Reference, Transform_7_#2:Transform) {is=title K2Node_BreakStruct}
   > Add(
    TargetArray=Get Control Rig Shape Definition()
    NewItem=Make Shape(ShapeName=@n13, Transform=InvertTransform(T=@n14.Transform_7_#2), ShapeProxy=@n14)
   )
   SetEditorProperty(
    execute=@n2.then_4
    Object=Get CR Shape Lib()
    PropertyName=shapes
    PropertyValue:Array of Shape Structures=Get Control Rig Shape Definition()
   )
   > Execute Python Script!(
    execute
    PythonScript="import unreal\r\nfound_key = None\r\nfor project_prefix, project_path in project_materials.items():\r\n    if shapelib.get_path_name().startswith(project_prefix):\r\n        found_key = project_prefix\r\n        break\r\nfound_key = found_key or '__default__'\r\n\r\nmat_path = project_materials[found_key]\r\nmat_asset = unreal.EditorAssetLibrary.find_asset_data(mat_path).get_asset()\r\nprint('mat asset: %s' % mat_asset)\r\nshapelib.set_editor_property('default_material', mat_asset)"
    shapelib:Control Rig Shape Library Object Reference=Get CR Shape Lib()
    project_materials:Map of Strings to Strings=Get Project Gizmo Materials() -> (Project Gizmo Materials:Map of Strings to Strings)
   ) -> (then, ReturnValue) {inputs [shapelib, project_materials]}
   > SaveLoadedAsset(AssetToSave=Get CR Shape Lib(), bOnlyIfIsDirty=false)
   Branch(execute=@n2.then_5, Condition=IsValid(Object=Get CRSL Maker File() -> (CRSL Maker File:CRSLM Blueprint Object Reference)))
   n15 = Get CRSL Maker File()
   > Set CR Shape Lib!(execute, CR Shape Lib:Control Rig Shape Library Object Reference=Get CR Shape Lib(), self:CRSLM Blueprint Object Reference=@n15) -> (then, Output_Get:Control Rig Shape Library Object Reference)
   > Set SKM!(execute, SKM:Skeletal Mesh Object Reference=Get SKM(), self:CRSLM Blueprint Object Reference=@n15) -> (then, Output_Get:Skeletal Mesh Object Reference)
   > Set Lego Model!(
    execute
    Lego Model:Atom Model Object Reference=Get Atom Model() -> (Atom Model:Atom Model Object Reference)
    self:CRSLM Blueprint Object Reference=@n15
   ) -> (then, Output_Get:Atom Model Object Reference)
   > Set Static Meshes!(execute, Static Meshes:Map of Names to Static Mesh Soft Object References=Get Static Meshes(), self:CRSLM Blueprint Object Reference=@n15) -> (then, Output_Get:Map of Names to Static Mesh Soft Object References)
   > Set Custom Shapes!(execute, Custom Shapes:Map of Names to Struct Custom Shape Structures=Get Custom Shapes(), self:CRSLM Blueprint Object Reference=@n15) -> (then, Output_Get:Map of Names to Struct Custom Shape Structures)
   > SaveLoadedAsset(AssetToSave=Get CRSL Maker File(), bOnlyIfIsDirty=false)
   Clear(execute=@n2.then_0, TargetArray:Array of Shape Structures=Get Control Rig Shape Definition())
   > Branch(Condition=Is Empty!(TargetMap:Map of Names to Static Mesh Soft Object References=Get Static Meshes()) -> (ReturnValue:bool) {is=title K2Node_CallFunction})
   > Populate Meshes()
   "On Clicked (Button_Populate_Meshes)"() {is=title K2Node_ComponentBoundEvent}
   > Populate Meshes()
   "On Property Changed (InputsDV)"!() -> (OutputDelegate, then, PropertyName:Name=None) {is=title K2Node_ComponentBoundEvent}
   > n16 = Is Valid(InputObject=Get CRSL Maker File())
   Branch(execute=@n16.Is Not Valid)
   n17 = Get Save Game() -> (Save Game:SG CRSL Maker Object Reference)
   > Set SKM(SKM=Get SKM(), self:SG CRSL Maker Object Reference=@n17)
   > Set CR Shape Lib(CR Shape Lib=Get CR Shape Lib(), self:SG CRSL Maker Object Reference=@n17)
   > Set Atom Model!(execute, Atom Model:Atom Model Object Reference=Get Atom Model(), self:SG CRSL Maker Object Reference=@n17) -> (then, Output_Get:Atom Model Object Reference)
   > Set Static Meshes(Static Meshes=Get Static Meshes(), self:SG CRSL Maker Object Reference=@n17)
   > AsyncSaveGameToSlot!(
    execute:Exec
    WorldContextObject:Object Reference
    SaveGameObject:Save Game Object Reference=@n17
    SlotName:String=CRSL_Maker
    UserIndex:int=0
   ) -> (then:Exec, Completed:Exec, SaveGame:Save Game Object Reference, bSuccess:bool=~) {is=title K2Node_AsyncAction}
   "On Clicked (Button_New)"() {is=title K2Node_ComponentBoundEvent}
   > n18 = GetAssetTools()
   > CreateAsset(
    self=@n18
    AssetName=CRSLM_Test
    PackagePath=/AnimInEdTools/CRSL_Maker
    AssetClass=&/CRSLM_Blueprint.CRSLM_Blueprint_C
   )
   "On Property Changed (CRSLMakerFileDV)"!() -> (OutputDelegate, then, PropertyName:Name=None) {is=title K2Node_ComponentBoundEvent}
   n19 = Get CRSL Maker File()
   > Set SKM!(execute, SKM=Get SKM!(self:CRSLM Blueprint Object Reference=@n19) -> (SKM)) -> (then, Output_Get)
   > Set CR Shape Lib!(execute, CR Shape Lib=Get CR Shape Lib!(self:CRSLM Blueprint Object Reference=@n19) -> (CR Shape Lib)) -> (then, Output_Get)
   > Set Atom Model!(execute, Atom Model=Get Lego Model!(self:CRSLM Blueprint Object Reference=@n19) -> (Lego Model:Atom Model Object Reference)) -> (then, Output_Get)
   > Set Static Meshes!(execute, Static Meshes=Get Static Meshes!(self:CRSLM Blueprint Object Reference=@n19) -> (Static Meshes)) -> (then, Output_Get)
   > Set Custom Shapes!(execute, Custom Shapes=Get Custom Shapes!(self:CRSLM Blueprint Object Reference=@n19) -> (Custom Shapes)) -> (then, Output_Get)
   Populate Meshes() {is=title K2Node_CustomEvent}
   n20 = Get Atom Model()
   > n21 = Is Valid(InputObject=@n20)
   > Set Selection Sets Index(Selection Sets Index:int=0) -> (Output_Get:int=0)
   > Clear!(execute, TargetMap:Map of Names to Static Mesh Soft Object References=Get Static Meshes!() -> (Static Meshes)) -> (then) {is=title K2Node_CallFunction}
   > n22 = For Each Loop(Array:Array of Soft Object References="Get (a ref)"!(
     Array:Array of Atom Model Processor Instance Structures=Get ModelProcessorInstances!(self:Atom Model Object Reference=@n20) -> (ModelProcessorInstances:Array of Atom Model Processor Instance Structures)
     Dimension 1=0
    ) -> (Output_bUseCustomSettings:bool=~, Output_ProcessedObjects:Array of Soft Object References).Output_ProcessedObjects) -> (Array Element:Soft Object Reference)
   > Branch(Condition=IsValidSoftObjectReference(SoftObjectReference=@n22))
   > n23 = LoadAsset_Blocking(Asset=@n22)
   n24 = EndsWith(SourceString=GetDisplayName(Object=@n23), InSuffix=_CP)
   > n25 = Branch(Condition=AND Boolean(A=NOT Boolean(A=@n24), B=IsValid(Object=@n23)))
   > n26 = Cast To StaticMesh(Object=@n23)
   n27 = Get Atom Model()
   > Add!(
    execute
    TargetMap:Map of Names to Static Mesh Soft Object References=Get Static Meshes()
    Key:"Name (by ref)"="Get (a ref)"!(
     Array:Array of Atom Model Selection Set Structures=Get SelectionSets!(self:Atom Model Object Reference=@n27) -> (SelectionSets:Array of Atom Model Selection Set Structures)
     Dimension 1=Get Selection Sets Index() -> (Selection Sets Index:int=0)
    ) -> (Output_PrimitiveIds:Set of Guid Structures, Output_SelectionSetName:Name).Output_SelectionSetName
    Value:"Static Mesh Soft Object Reference (by ref)"=Make Soft Reference(Input:Static Mesh Object Reference=@n26) -> (Output:Static Mesh Soft Object Reference)
   ) -> (then) {is=title K2Node_CallFunction}
   > Set Selection Sets Index(Selection Sets Index=int + int(A=Get Selection Sets Index() -> (Selection Sets Index=0))) -> (Output_Get=0)
   n28 = GetSoftPtrAssetName!(Asset:Soft Object Reference=@n22) -> (ReturnValue:String)
   PrintText(
    execute=@n26.CastFailed
    InText=Format Text!(Format="LOCTEXT(\"CRSL_Maker : Couldn\\'t cast {Object} to StaticMesh\")", Object:String=@n28) -> (Result)
   )
   PrintText(execute=@n25.else, InText=Format Text(Format="LOCTEXT(\"CRSL_Maker : Couldn\\'t load {Object}\")", Object=@n28))
   PrintString(execute=@n21.Is Not Valid, InString=CRSL_Maker : The Atom Model provided is not valid.)
   > n29 = Construct Object from Class(Class=&/EUW_CRSL_AddSelectedSM.EUW_CRSL_AddSelectedSM_C) -> (ReturnValue:EUW CRSL Add Selected SM Object Reference)
   > AddChild(self=Get DialogSlot() -> (DialogSlot:Named Slot Object Reference), Content=@n29)
   > Bind Event to Static Mesh Map(self:EUW CRSL Add Selected SM Object Reference=@n29, Delegate=Create Event()) {is=title K2Node_AddDelegate}
   Event PreConstruct()
   > SetObject(
    self=[
     Get InputsDV() -> (InputsDV:Details View Object Reference)
     Get CRSLMakerFileDV() -> (CRSLMakerFileDV:Details View Object Reference)
     Get InputsDV()
    ]
    NewObject=Self-Reference()
   )
   n30 = Event Set Asset Breadcrumb!() -> (OutputDelegate, then, Bread Crumb:ABT Abstract Object Reference) {is=title K2Node_Event}
   > n31 = Bad cast node!(execute, Object=@n30.Bread Crumb) -> (then, CastFailed, AsABT Creature:Object Reference) {is=title K2Node_DynamicCast}
   > n32 = Set Asset Breadcrumb(Asset Breadcrumb:Object Reference=@n31) -> (Output_Get:Object Reference)
   > Set Atom Model(Atom Model=Resolve Soft Reference!(Input:Atom Model Soft Object Reference=Get Atom Model!(self:Object Reference=@n32) -> (Atom Model:Atom Model Soft Object Reference)) -> (Output:Atom Model Object Reference))
   > n33 = Sequence!(execute) -> (then_0, then_1)
   > n34 = LoadAsset_Blocking(Asset=Get Skeletal Mesh!(self:Object Reference=@n32) -> (Skeletal Mesh:Skeletal Mesh Soft Object Reference))
   > n35 = Cast To SkeletalMesh!(execute, Object=@n34) -> (then, CastFailed, AsSkeletal Mesh:Skeletal Mesh Object Reference) {is=title K2Node_DynamicCast}
   > Set SKM(SKM=@n35)
   n36 = LoadAsset_Blocking(
    execute=@n33.then_1
    Asset=Get Control Rig Shape Library!(self:Object Reference=@n32) -> (Control Rig Shape Library:Control Rig Shape Library Soft Object Reference)
   )
   > n37 = Cast To ControlRigShapeLibrary!(execute, Object=@n36) -> (then, CastFailed, AsControl Rig Shape Library:Control Rig Shape Library Object Reference) {is=title K2Node_DynamicCast}
   > Set CR Shape Lib(CR Shape Lib=@n37)
   Event Construct()
   > n38 = LoadGameFromSlot!(execute, SlotName:String=CRSL_Maker, UserIndex:int=0) -> (then, ReturnValue:Save Game Object Reference)
   > n39 = Cast To SG_CRSL_Maker!(execute, Object=@n38) -> (then, CastFailed, AsSG CRSL Maker:SG CRSL Maker Object Reference) {is=title K2Node_DynamicCast}
   > n40 = Sequence!(execute) -> (then_0, then_1, then_2:Exec, then_3:Exec, then_4:Exec)
   > Set Save Game(Save Game:SG CRSL Maker Object Reference=@n39) -> (Output_Get:SG CRSL Maker Object Reference)
   n41 = Get CR Shape Lib(self:SG CRSL Maker Object Reference=@n39)
   Is Valid(exec=@n40.then_1, InputObject=@n41)
   > Set CR Shape Lib(CR Shape Lib=@n41)
   n42 = Get SKM(self:SG CRSL Maker Object Reference=@n39)
   Is Valid(exec=@n40.then_2, InputObject=@n42)
   > Set SKM(SKM=@n42)
   n43 = Get Atom Model(self:SG CRSL Maker Object Reference=@n39) -> (Atom Model:Atom Model Object Reference)
   Is Valid(exec=@n40.then_3, InputObject=@n43)
   > Set Atom Model(Atom Model=@n43)
   n44 = Get Static Meshes!(self:SG CRSL Maker Object Reference=@n39) -> (Static Meshes)
   Branch(
    execute=@n40.then_4
    Condition=Is Not Empty!(TargetMap:Map of Names to Static Mesh Soft Object References=@n44) -> (ReturnValue:bool) {is=title K2Node_CallFunction}
   )
   > Set Static Meshes(Static Meshes=@n44)
   n45 = CreateSaveGameObject!(execute=@n39.CastFailed, SaveGameClass:Save Game Class Reference=/AnimInEdTools/CRSL_Maker/Data/SG_CRSL_Maker.SG_CRSL_Maker_C) -> (then, ReturnValue:SG CRSL Maker Object Reference)
   > Set Save Game(Save Game=@n45)
   Execute Python Script!(
    execute
    PythonScript="import unreal\r\nmat_asset = unreal.EditorAssetLibrary.find_asset_data('&M_ControlRigGizmo').get_asset()\r\nshapelib.set_editor_property('default_material', mat_asset)"
    shapelib=Get CR Shape Lib!() -> (CR Shape Lib)
   ) -> (then, ReturnValue) {inputs [shapelib]}
   Reroute Node(InputPin:int=@n22.Array Index) -> (OutputPin:int)
  ]}
 Set SM Map from Widget {nodes [
   Set SM Map from Widget!() -> (then, SM Map:Map of Static Mesh Object References to Names) {is=title K2Node_FunctionEntry}
   > Clear(TargetMap=Get Buffer Static Meshes() -> (Buffer Static Meshes:Map of Names to Static Mesh Soft Object References))
   > n1 = Sequence!(execute) -> (then_0, then_1)
   > n2 = Branch(Condition=false)
   > Clear(TargetMap=Get Static Meshes!() -> (Static Meshes))
   n3 = "For Each Loop (Map)"!(execute:Exec=@n2.else, MapPin:Map of Names to Static Mesh Soft Object References=Get Static Meshes(), BreakPin:Exec) -> (then:Exec, KeyPin:Name, ValuePin:Static Mesh Soft Object Reference, CompletedPin:Exec)
   > Add(TargetMap=Get Buffer Static Meshes(), Key=@n3, Value=@n3.ValuePin)
   n4 = "For Each Loop (Map)"(
    execute=@n1.then_1
    MapPin:Map of Static Mesh Object References to Names=Get SM Map() -> (SM Map:Map of Static Mesh Object References to Names)
   ) -> (KeyPin:Static Mesh Object Reference, ValuePin:Name)
   > Add(TargetMap=Get Buffer Static Meshes(), Key=@n4.ValuePin, Value=Make Soft Reference(Input=@n4))
   Set Static Meshes(execute=@n4.CompletedPin, Static Meshes=Get Buffer Static Meshes())
   IsChecked(self=Get ReplaceCheckbox() -> (ReplaceCheckbox:Editor Utility Check Box Object Reference))
  ]}
]
variables [
 Asset Breadcrumb {type=Object, default=null}
 Atom Model {type=AtomModel, default=null}
 Control Rig Shape Definition {type="array<ControlRigShapeDefinition>", default []}
 CR Shape Lib {type=ControlRigShapeLibrary, default=null}
 CRSL Maker File {type=CRSLM_Blueprint_C, default=null}
 Custom Shapes {type="map<Struct_CustomShape>", default {}}
 Dynamic Mesh {type=DynamicMesh, default=null}
 Dynamic Mesh Pool {type=DynamicMeshPool, default=null}
 Project Gizmo Materials {
  type="map<string>"
  default {
   /Juno=&M_ControlRigGizmo
   /Sprout=&M_ControlRigGizmo
   __default__=&/M_ControlRigGizmo_Transparent
  }
 }
 Save Game {type=SG_CRSL_Maker_C, default=null}
 Selection Sets Index {type=integer}
 SKM {type=SkeletalMesh, default=null}
 Static Meshes {type="map<StaticMesh>", default {}}
]
widgets {
 info {rootWidgetClass=&CanvasPanel, widgetCount=13}
 widgets [
  CanvasPanel_0 {widgetClassPath=&CanvasPanel, widget=CanvasPanel_0}
  Overlay_293 {
   widgetClassPath=&Overlay
   parent=CanvasPanel_0
   widget=Overlay_293
   slot {class=CanvasPanelSlot, properties {zOrder=1, layoutData {alignment=<0 0>, anchors {maximum=<1 1>, minimum=<0 0>}, offsets=<ltrb 0 0 0 0>}}}
  }
  DialogSlot {
   widgetClassPath=&NamedSlot
   bIsVariable=true
   parent=Overlay_293
   widget=DialogSlot
   slot {class=OverlaySlot, properties {horizontalAlignment=HAlign_Fill, verticalAlignment=VAlign_Fill}}
  }
  VerticalBox_0 {
   widgetClassPath=&VerticalBox
   parent=CanvasPanel_0
   widget=VerticalBox_0
   slot {
    class=CanvasPanelSlot
    properties {bAutoSize=true, layoutData {alignment=<0 0>, anchors {maximum=<1 1>, minimum=<0 0>}, offsets=<ltrb 0 0 0 0>}}
   }
  }
  ScrollBox_0 {
   widgetClassPath=&ScrollBox
   parent=VerticalBox_0
   widget=ScrollBox_0
   properties {
    widgetBarStyle {
     thickness=8
     draggedThumbImage {
      bIsDynamicallyLoaded=false
      drawAs=RoundedBox
      imageType=NoImage
      mirroring=NoMirror
      resourceName=None
      resourceObject=null
      tiling=NoTile
      imageSize=<0 0>
      margin=<ltrb 0 0 0 0>
      outlineSettings {
       bUseBrushTransparency=false
       roundingType=FixedRadius
       width=0
       color {colorUseRule=UseColor_Specified, specifiedColor=<rgba 0 0 0 0>}
       cornerRadii=<4 4 4 4>
      }
      tintColor {colorUseRule=UseColor_Specified, specifiedColor=<rgba 0.2159 0.2159 0.2159 1>}
      uVRegion {bIsValid=false, max=<0 0>, min=<0 0>}
     }
     horizontalBackgroundImage {
      bIsDynamicallyLoaded=false
      drawAs=NoDrawType
      imageType=NoImage
      mirroring=NoMirror
      resourceName=None
      resourceObject=null
      tiling=NoTile
      imageSize=<0 0>
      margin=<ltrb 0 0 0 0>
      outlineSettings {
       bUseBrushTransparency=false
       roundingType=HalfHeightRadius
       width=0
       color {colorUseRule=UseColor_Specified, specifiedColor=<rgba 0 0 0 0>}
       cornerRadii=<0 0 0 0>
      }
      tintColor {colorUseRule=UseColor_Specified, specifiedColor=<rgba 1 1 1 1>}
      uVRegion {bIsValid=false, max=<0 0>, min=<0 0>}
     }
     horizontalBottomSlotImage {
      bIsDynamicallyLoaded=false
      drawAs=NoDrawType
      imageType=NoImage
      mirroring=NoMirror
      resourceName=None
      resourceObject=null
      tiling=NoTile
      imageSize=<0 0>
      margin=<ltrb 0 0 0 0>
      outlineSettings {
       bUseBrushTransparency=false
       roundingType=HalfHeightRadius
       width=0
       color {colorUseRule=UseColor_Specified, specifiedColor=<rgba 0 0 0 0>}
       cornerRadii=<0 0 0 0>
      }
      tintColor {colorUseRule=UseColor_Specified, specifiedColor=<rgba 1 1 1 1>}
      uVRegion {bIsValid=false, max=<0 0>, min=<0 0>}
     }
     horizontalTopSlotImage {
      bIsDynamicallyLoaded=false
      drawAs=NoDrawType
      imageType=NoImage
      mirroring=NoMirror
      resourceName=None
      resourceObject=null
      tiling=NoTile
      imageSize=<0 0>
      margin=<ltrb 0 0 0 0>
      outlineSettings {
       bUseBrushTransparency=false
       roundingType=HalfHeightRadius
       width=0
       color {colorUseRule=UseColor_Specified, specifiedColor=<rgba 0 0 0 0>}
       cornerRadii=<0 0 0 0>
      }
      tintColor {colorUseRule=UseColor_Specified, specifiedColor=<rgba 1 1 1 1>}
      uVRegion {bIsValid=false, max=<0 0>, min=<0 0>}
     }
     hoveredThumbImage {
      bIsDynamicallyLoaded=false
      drawAs=RoundedBox
      imageType=NoImage
      mirroring=NoMirror
      resourceName=None
      resourceObject=null
      tiling=NoTile
      imageSize=<0 0>
      margin=<ltrb 0 0 0 0>
      outlineSettings {
       bUseBrushTransparency=false
       roundingType=FixedRadius
       width=0
       color {colorUseRule=UseColor_Specified, specifiedColor=<rgba 0 0 0 0>}
       cornerRadii=<4 4 4 4>
      }
      tintColor {colorUseRule=UseColor_Specified, specifiedColor=<rgba 0.2159 0.2159 0.2159 1>}
      uVRegion {bIsValid=false, max=<0 0>, min=<0 0>}
     }
     normalThumbImage {
      bIsDynamicallyLoaded=false
      drawAs=RoundedBox
      imageType=NoImage
      mirroring=NoMirror
      resourceName=None
      resourceObject=null
      tiling=NoTile
      imageSize=<0 0>
      margin=<ltrb 0 0 0 0>
      outlineSettings {
       bUseBrushTransparency=false
       roundingType=FixedRadius
       width=0
       color {colorUseRule=UseColor_Specified, specifiedColor=<rgba 0 0 0 0>}
       cornerRadii=<4 4 4 4>
      }
      tintColor {colorUseRule=UseColor_Specified, specifiedColor=<rgba 0.0953 0.0953 0.0953 1>}
      uVRegion {bIsValid=false, max=<0 0>, min=<0 0>}
     }
     verticalBackgroundImage {
      bIsDynamicallyLoaded=false
      drawAs=NoDrawType
      imageType=NoImage
      mirroring=NoMirror
      resourceName=None
      resourceObject=null
      tiling=NoTile
      imageSize=<0 0>
      margin=<ltrb 0 0 0 0>
      outlineSettings {
       bUseBrushTransparency=false
       roundingType=HalfHeightRadius
       width=0
       color {colorUseRule=UseColor_Specified, specifiedColor=<rgba 0 0 0 0>}
       cornerRadii=<0 0 0 0>
      }
      tintColor {colorUseRule=UseColor_Specified, specifiedColor=<rgba 1 1 1 1>}
      uVRegion {bIsValid=false, max=<0 0>, min=<0 0>}
     }
     verticalBottomSlotImage {
      bIsDynamicallyLoaded=false
      drawAs=NoDrawType
      imageType=NoImage
      mirroring=NoMirror
      resourceName=None
      resourceObject=null
      tiling=NoTile
      imageSize=<0 0>
      margin=<ltrb 0 0 0 0>
      outlineSettings {
       bUseBrushTransparency=false
       roundingType=HalfHeightRadius
       width=0
       color {colorUseRule=UseColor_Specified, specifiedColor=<rgba 0 0 0 0>}
       cornerRadii=<0 0 0 0>
      }
      tintColor {colorUseRule=UseColor_Specified, specifiedColor=<rgba 1 1 1 1>}
      uVRegion {bIsValid=false, max=<0 0>, min=<0 0>}
     }
     verticalTopSlotImage {
      bIsDynamicallyLoaded=false
      drawAs=NoDrawType
      imageType=NoImage
      mirroring=NoMirror
      resourceName=None
      resourceObject=null
      tiling=NoTile
      imageSize=<0 0>
      margin=<ltrb 0 0 0 0>
      outlineSettings {
       bUseBrushTransparency=false
       roundingType=HalfHeightRadius
       width=0
       color {colorUseRule=UseColor_Specified, specifiedColor=<rgba 0 0 0 0>}
       cornerRadii=<0 0 0 0>
      }
      tintColor {colorUseRule=UseColor_Specified, specifiedColor=<rgba 1 1 1 1>}
      uVRegion {bIsValid=false, max=<0 0>, min=<0 0>}
     }
    }
    widgetStyle {
     barThickness=8
     bottomShadowBrush {
      bIsDynamicallyLoaded=false
      drawAs=Box
      imageType=FullColor
      mirroring=NoMirror
      resourceName=../../../Engine/Content/Slate/Common/ScrollBoxShadowBottom.png
      resourceObject=null
      tiling=NoTile
      imageSize=<16 8>
      margin=<ltrb 0.5 0 0.5 1>
      outlineSettings {
       bUseBrushTransparency=false
       roundingType=HalfHeightRadius
       width=0
       color {colorUseRule=UseColor_Specified, specifiedColor=<rgba 0 0 0 0>}
       cornerRadii=<0 0 0 0>
      }
      tintColor {colorUseRule=UseColor_Specified, specifiedColor=<rgba 1 1 1 1>}
      uVRegion {bIsValid=false, max=<0 0>, min=<0 0>}
     }
     horizontalScrolledContentPadding=<ltrb 0 0 1 0>
     leftShadowBrush {
      bIsDynamicallyLoaded=false
      drawAs=Box
      imageType=FullColor
      mirroring=NoMirror
      resourceName=../../../Engine/Content/Slate/Common/ScrollBoxShadowLeft.png
      resourceObject=null
      tiling=NoTile
      imageSize=<8 16>
      margin=<ltrb 1 0.5 0 0.5>
      outlineSettings {
       bUseBrushTransparency=false
       roundingType=HalfHeightRadius
       width=0
       color {colorUseRule=UseColor_Specified, specifiedColor=<rgba 0 0 0 0>}
       cornerRadii=<0 0 0 0>
      }
      tintColor {colorUseRule=UseColor_Specified, specifiedColor=<rgba 1 1 1 1>}
      uVRegion {bIsValid=false, max=<0 0>, min=<0 0>}
     }
     rightShadowBrush {
      bIsDynamicallyLoaded=false
      drawAs=Box
      imageType=FullColor
      mirroring=NoMirror
      resourceName=../../../Engine/Content/Slate/Common/ScrollBoxShadowRight.png
      resourceObject=null
      tiling=NoTile
      imageSize=<8 16>
      margin=<ltrb 0 0.5 1 0.5>
      outlineSettings {
       bUseBrushTransparency=false
       roundingType=HalfHeightRadius
       width=0
       color {colorUseRule=UseColor_Specified, specifiedColor=<rgba 0 0 0 0>}
       cornerRadii=<0 0 0 0>
      }
      tintColor {colorUseRule=UseColor_Specified, specifiedColor=<rgba 1 1 1 1>}
      uVRegion {bIsValid=false, max=<0 0>, min=<0 0>}
     }
     topShadowBrush {
      bIsDynamicallyLoaded=false
      drawAs=Box
      imageType=FullColor
      mirroring=NoMirror
      resourceName=../../../Engine/Content/Slate/Common/ScrollBoxShadowTop.png
      resourceObject=null
      tiling=NoTile
      imageSize=<16 8>
      margin=<ltrb 0.5 1 0.5 0>
      outlineSettings {
       bUseBrushTransparency=false
       roundingType=HalfHeightRadius
       width=0
       color {colorUseRule=UseColor_Specified, specifiedColor=<rgba 0 0 0 0>}
       cornerRadii=<0 0 0 0>
      }
      tintColor {colorUseRule=UseColor_Specified, specifiedColor=<rgba 1 1 1 1>}
      uVRegion {bIsValid=false, max=<0 0>, min=<0 0>}
     }
     verticalScrolledContentPadding=<ltrb 0 0 0 1>
    }
   }
   slot {class=VerticalBoxSlot, properties {size {sizeRule=Fill, value=1}}}
  }
  Button_New {
   widgetClassPath=&EditorUtilityButton
   parent=ScrollBox_0
   widget=Button_New
   properties {visibility=Collapsed}
   slot {class=ScrollBoxSlot, properties {horizontalAlignment=HAlign_Left}}
  }
  TextBlock_0 {widgetClassPath=&TextBlock, parent=Button_New, widget=TextBlock_0, properties {text=+}, slot {class=ButtonSlot}}
  CRSLMakerFileDV {
   widgetClassPath=&DetailsView
   parent=ScrollBox_0
   widget=CRSLMakerFileDV
   properties {categoriesToShow [File]}
   slot {class=ScrollBoxSlot}
  }
  InputsDV {
   widgetClassPath=&DetailsView
   parent=ScrollBox_0
   widget=InputsDV
   properties {categoriesToShow [Inputs]}
   slot {class=ScrollBoxSlot}
  }
  Button_Populate_Meshes {
   widgetClassPath=&Button
   parent=VerticalBox_0
   widget=Button_Populate_Meshes
   properties {widgetStyle {
     clickedSlateSound {resourceObject=null}
     disabled {
      bIsDynamicallyLoaded=false
      drawAs=RoundedBox
      imageType=NoImage
      mirroring=NoMirror
      resourceName=None
      resourceObject=null
      tiling=NoTile
      imageSize=<0 0>
      margin=<ltrb 0 0 0 0>
      outlineSettings {
       bUseBrushTransparency=false
       roundingType=FixedRadius
       width=1
       color {colorUseRule=UseColor_ColorTable, specifiedColor=<rgba 0 0 0 0>}
       cornerRadii=<4 4 4 4>
      }
      tintColor {colorUseRule=UseColor_Specified, specifiedColor=<rgba 0.0395 0.0395 0.0395 1>}
      uVRegion {bIsValid=false, max=<0 0>, min=<0 0>}
     }
     disabledForeground {colorUseRule=UseColor_ColorTable, specifiedColor=<rgba 0 0 0 0>}
     hovered {
      bIsDynamicallyLoaded=false
      drawAs=RoundedBox
      imageType=NoImage
      mirroring=NoMirror
      resourceName=None
      resourceObject=null
      tiling=NoTile
      imageSize=<32 32>
      margin=<ltrb 0 0 0 0>
      outlineSettings {
       bUseBrushTransparency=false
       roundingType=FixedRadius
       width=1
       color {colorUseRule=UseColor_ColorTable, specifiedColor=<rgba 0 0 0 0>}
       cornerRadii=<4 4 4 4>
      }
      tintColor {colorUseRule=UseColor_Specified, specifiedColor=<rgba 0.0953 0.0953 0.0953 1>}
      uVRegion {bIsValid=false, max=<0 0>, min=<0 0>}
     }
     hoveredForeground {colorUseRule=UseColor_ColorTable, specifiedColor=<rgba 0 0 0 0>}
     hoveredSlateSound {resourceObject=null}
     normal {
      bIsDynamicallyLoaded=false
      drawAs=RoundedBox
      imageType=NoImage
      mirroring=NoMirror
      resourceName=None
      resourceObject=null
      tiling=NoTile
      imageSize=<32 32>
      margin=<ltrb 0 0 0 0>
      outlineSettings {
       bUseBrushTransparency=false
       roundingType=FixedRadius
       width=1
       color {colorUseRule=UseColor_ColorTable, specifiedColor=<rgba 0 0 0 0>}
       cornerRadii=<4 4 4 4>
      }
      tintColor {colorUseRule=UseColor_Specified, specifiedColor=<rgba 0.0395 0.0395 0.0395 1>}
      uVRegion {bIsValid=false, max=<0 0>, min=<0 0>}
     }
     normalForeground {colorUseRule=UseColor_ColorTable, specifiedColor=<rgba 0 0 0 0>}
     normalPadding=<ltrb 12 1.5 12 1.5>
     pressed {
      bIsDynamicallyLoaded=false
      drawAs=RoundedBox
      imageType=NoImage
      mirroring=NoMirror
      resourceName=None
      resourceObject=null
      tiling=NoTile
      imageSize=<32 32>
      margin=<ltrb 0 0 0 0>
      outlineSettings {
       bUseBrushTransparency=false
       roundingType=FixedRadius
       width=1
       color {colorUseRule=UseColor_ColorTable, specifiedColor=<rgba 0 0 0 0>}
       cornerRadii=<4 4 4 4>
      }
      tintColor {colorUseRule=UseColor_Specified, specifiedColor=<rgba 0.0284 0.0284 0.0284 1>}
      uVRegion {bIsValid=false, max=<0 0>, min=<0 0>}
     }
     pressedForeground {colorUseRule=UseColor_ColorTable, specifiedColor=<rgba 0 0 0 0>}
     pressedPadding=<ltrb 12 2.5 12 0.5>
     pressedSlateSound {resourceObject=null}
    }}
   slot {class=VerticalBoxSlot}
  }
  TextBlock {
   widgetClassPath=&TextBlock
   parent=Button_Populate_Meshes
   widget=TextBlock
   properties {
    autoWrapText=true
    justification=Center
    text=Populate Meshes
    colorAndOpacity {colorUseRule=UseColor_Specified, specifiedColor=<rgba 0.5255 0.5255 0.5255 1>}
    font {
     bForceMonospaced=false
     bMaterialIsStencil=false
     fontMaterial=null
     fontObject=&Roboto
     letterSpacing=0
     monospacedWidth=1
     size=10
     skewAmount=0
     typefaceFontName=Bold
     outlineSettings {
      bApplyOutlineToDropShadows=false
      bMiteredCorners=false
      bSeparateFillAlpha=false
      outlineMaterial=null
      outlineSize=0
      outlineColor=<rgba 0 0 0 1>
     }
    }
   }
   slot {class=ButtonSlot}
  }
  Button_Run {
   widgetClassPath=&Button
   parent=VerticalBox_0
   widget=Button_Run
   properties {widgetStyle {
     clickedSlateSound {resourceObject=null}
     disabled {
      bIsDynamicallyLoaded=false
      drawAs=RoundedBox
      imageType=NoImage
      mirroring=NoMirror
      resourceName=None
      resourceObject=null
      tiling=NoTile
      imageSize=<0 0>
      margin=<ltrb 0 0 0 0>
      outlineSettings {
       bUseBrushTransparency=false
       roundingType=FixedRadius
       width=1
       color {colorUseRule=UseColor_ColorTable, specifiedColor=<rgba 0 0 0 0>}
       cornerRadii=<4 4 4 4>
      }
      tintColor {colorUseRule=UseColor_Specified, specifiedColor=<rgba 0.0395 0.0395 0.0395 1>}
      uVRegion {bIsValid=false, max=<0 0>, min=<0 0>}
     }
     disabledForeground {colorUseRule=UseColor_ColorTable, specifiedColor=<rgba 0 0 0 0>}
     hovered {
      bIsDynamicallyLoaded=false
      drawAs=RoundedBox
      imageType=NoImage
      mirroring=NoMirror
      resourceName=None
      resourceObject=null
      tiling=NoTile
      imageSize=<32 32>
      margin=<ltrb 0 0 0 0>
      outlineSettings {
       bUseBrushTransparency=false
       roundingType=FixedRadius
       width=1
       color {colorUseRule=UseColor_ColorTable, specifiedColor=<rgba 0 0 0 0>}
       cornerRadii=<4 4 4 4>
      }
      tintColor {colorUseRule=UseColor_Specified, specifiedColor=<rgba 0.0953 0.0953 0.0953 1>}
      uVRegion {bIsValid=false, max=<0 0>, min=<0 0>}
     }
     hoveredForeground {colorUseRule=UseColor_ColorTable, specifiedColor=<rgba 0 0 0 0>}
     hoveredSlateSound {resourceObject=null}
     normal {
      bIsDynamicallyLoaded=false
      drawAs=RoundedBox
      imageType=NoImage
      mirroring=NoMirror
      resourceName=None
      resourceObject=null
      tiling=NoTile
      imageSize=<32 32>
      margin=<ltrb 0 0 0 0>
      outlineSettings {
       bUseBrushTransparency=false
       roundingType=FixedRadius
       width=1
       color {colorUseRule=UseColor_ColorTable, specifiedColor=<rgba 0 0 0 0>}
       cornerRadii=<4 4 4 4>
      }
      tintColor {colorUseRule=UseColor_Specified, specifiedColor=<rgba 0.0395 0.0395 0.0395 1>}
      uVRegion {bIsValid=false, max=<0 0>, min=<0 0>}
     }
     normalForeground {colorUseRule=UseColor_ColorTable, specifiedColor=<rgba 0 0 0 0>}
     normalPadding=<ltrb 12 1.5 12 1.5>
     pressed {
      bIsDynamicallyLoaded=false
      drawAs=RoundedBox
      imageType=NoImage
      mirroring=NoMirror
      resourceName=None
      resourceObject=null
      tiling=NoTile
      imageSize=<32 32>
      margin=<ltrb 0 0 0 0>
      outlineSettings {
       bUseBrushTransparency=false
       roundingType=FixedRadius
       width=1
       color {colorUseRule=UseColor_ColorTable, specifiedColor=<rgba 0 0 0 0>}
       cornerRadii=<4 4 4 4>
      }
      tintColor {colorUseRule=UseColor_Specified, specifiedColor=<rgba 0.0284 0.0284 0.0284 1>}
      uVRegion {bIsValid=false, max=<0 0>, min=<0 0>}
     }
     pressedForeground {colorUseRule=UseColor_ColorTable, specifiedColor=<rgba 0 0 0 0>}
     pressedPadding=<ltrb 12 2.5 12 0.5>
     pressedSlateSound {resourceObject=null}
    }}
   slot {class=VerticalBoxSlot}
  }
  TextBlock_1 {
   widgetClassPath=&TextBlock
   parent=Button_Run
   widget=TextBlock_1
   properties {
    autoWrapText=true
    justification=Center
    text=Run
    colorAndOpacity {colorUseRule=UseColor_Specified, specifiedColor=<rgba 0.5255 0.5255 0.5255 1>}
    font {
     bForceMonospaced=false
     bMaterialIsStencil=false
     fontMaterial=null
     fontObject=&Roboto
     letterSpacing=0
     monospacedWidth=1
     size=10
     skewAmount=0
     typefaceFontName=Bold
     outlineSettings {
      bApplyOutlineToDropShadows=false
      bMiteredCorners=false
      bSeparateFillAlpha=false
      outlineMaterial=null
      outlineSize=0
      outlineColor=<rgba 0 0 0 1>
     }
    }
   }
   slot {class=ButtonSlot}
  }
 ]
}
